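; Smart object behavior definition for the dragon cart's large door trigger, deciding whether a user can use it.
; It overrides CanBeUsedBy: after the parent check passes, it casts the smart object actor to the large door trigger and scans its target doors.
; Use is allowed only while at least one of those doors is not open; a failed parent check or cast returns false.
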(fn CanBeUsedBy
  (params (User :type Actor%20Object%20Reference) (SmartObjectActor :type Actor%20Object%20Reference))
  (if (|CallParentFunction User SmartObjectActor)
    (bind _asbp_dragon_cart_large_door_trigger (Utilities|Casting|CastToBP_DragonCart_LargeDoorTrigger (Variables|GetSmartObjectActor))
      (:then
        (for _array_element (Variables|/DragonCartDoors/Gameplay/BP_DragonCart_LargeDoorTrigger.BP_DragonCart_LargeDoorTrigger_C|GetTargetDoors _asbp_dragon_cart_large_door_trigger)
          (if (not (Class|BPLargeDoorParent|IsLargeDoorOpen _array_element))
            (return true)))
        (return false))
      (:CastFailed
        _))
    (else
      _))
  (return false))
