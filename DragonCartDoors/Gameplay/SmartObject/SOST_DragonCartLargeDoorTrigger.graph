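# Smart object State Tree for the DragonCart large door trigger, where an AI agent performs a simple interaction and then pauses.
# The Interact state runs the Fortnite AI interact task and moves to Wait when it completes.
# Wait holds for a short randomized delay of around two seconds, then ends the tree as succeeded.
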StateTree /DragonCartDoors/Gameplay/SmartObject/SOST_DragonCartLargeDoorTrigger
states [
 Root {
  id=#1
  type=STATE
  children [#2, #3]
  tasks []
  transitions []
 }
 Interact {
  id=#2
  type=STATE
  tasks [{
    id=#4
    description="(Fortnite) Interact Task"
    properties {
     iD=#5
     instance {
      _structType=/Script/FortniteAIServer.FortAthenaStateTreeInteractTaskInstanceData
      actor=null
      interactDuration=0
      interactTargetActor=null
      interactType=IT_Simple
      timer=0
     }
     node {
      name=None
      _structType=/Script/FortniteAIServer.FortAthenaStateTreeInteractTask
      bCanEditConsideredForCompletion=true
      bConsideredForCompletion=true
      bHasTaskCompletionDelegateDispatcher=false
      bTaskEnabled=true
      transitionHandlingPriority=Normal
      bindingsBatch {value=65535}
      executionRuntimeTemplateIndex {value=65535}
      instanceDataHandle {index=65535, source=None, stateHandle {index=65535}}
      instanceTemplateIndex {value=65535}
      outputBindingsBatch {value=65535}
     }
    }
   }]
  transitions [{id=#6, trigger=ON_STATE_COMPLETED, state {id=#3, name=Wait}}]
 }
 Wait {
  id=#3
  type=STATE
  tasks [{
    id=#7
    description="Delay 2.0 ±0.5"
    properties {
     iD=#8
     instance {_structType=&StateTreeDelayTaskInstanceData, duration=2, randomDeviation=0.5}
     node {name=None, _structType=&StateTreeDelayTask}
    }
   }]
  transitions [{id=#9, trigger=ON_STATE_COMPLETED, state {id=#10, name=Root, linkType=SUCCEEDED}}]
 }
]
